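# Sound cue for Grecian urn demolition impacts, classed as an impact sound with a concurrency rule that limits simultaneous instances.
# A quality-level root picks one of four mixers, with smaller variation pools in the last two.
# Each mixer layers randomized close urn impacts with distant rock impacts through separate attenuations that crossfade by distance.
# Modulators on every layer vary playback, including widened pitch ranges on the distant rock hits.
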SoundCue /Game/Sounds/Fort_Impact_Sounds/Demolition/Urn/Demo_Grecian_Urn_Impact_MLTI_Cue
root=QualityLevel_0
nodes [
 {
  id=Attenuation_0
  children [Modulator_1]
  properties {
   attenuationSettings=/Game/Sounds/Attenuation/Fort_Demolition_Crossfade_Global_Hit_Far_Att
   attenuationOverrides {falloffDistance=3000, attenuationShapeExtents {x=1000}}
  }
 }
 {
  id=Attenuation_1
  children [Modulator_5]
  properties {
   attenuationSettings=/Game/Sounds/Attenuation/Fort_Demolition_Crossfade_Global_Hit_Close_Att
   attenuationOverrides {falloffDistance=900, attenuationShapeExtents {x=100}}
  }
 }
 {
  id=Attenuation_10
  children [Modulator_11]
  properties {
   attenuationSettings=&Fort_Demolition_Crossfade_Global_Hit_Close_Att
   attenuationOverrides {falloffDistance=900, attenuationShapeExtents {x=100}}
  }
 }
 {
  id=Attenuation_3
  children [Modulator_4]
  properties {
   attenuationSettings=&Fort_Demolition_Crossfade_Global_Hit_Far_Att
   attenuationOverrides {falloffDistance=3000, attenuationShapeExtents {x=1000}}
  }
 }
 {
  id=Attenuation_4
  children [Modulator_2]
  properties {
   attenuationSettings=&Fort_Demolition_Crossfade_Global_Hit_Close_Att
   attenuationOverrides {falloffDistance=900, attenuationShapeExtents {x=100}}
  }
 }
 {
  id=Attenuation_5
  children [Modulator_7]
  properties {
   attenuationSettings=&Fort_Demolition_Crossfade_Global_Hit_Far_Att
   attenuationOverrides {falloffDistance=3000, attenuationShapeExtents {x=1000}}
  }
 }
 {
  id=Attenuation_7
  children [Modulator_10]
  properties {
   attenuationSettings=&Fort_Demolition_Crossfade_Global_Hit_Far_Att
   attenuationOverrides {falloffDistance=3000, attenuationShapeExtents {x=1000}}
  }
 }
 {
  id=Attenuation_8
  children [Modulator_8]
  properties {
   attenuationSettings=&Fort_Demolition_Crossfade_Global_Hit_Close_Att
   attenuationOverrides {falloffDistance=900, attenuationShapeExtents {x=100}}
  }
 }
 {id=Mixer_1, children [Attenuation_4, Attenuation_0], properties {inputVolume [1,1]}}
 {id=Mixer_3, children [Attenuation_1, Attenuation_3], properties {inputVolume [1,0.8]}}
 {id=Mixer_5, children [Attenuation_8, Attenuation_5], properties {inputVolume [1,1]}}
 {id=Mixer_7, children [Attenuation_10, Attenuation_7], properties {inputVolume [1,1]}}
 {id=Modulator_1, children [WavePlayer_1], properties {pitchMax=1.4}}
 {id=Modulator_10, children [WavePlayer_15], properties {pitchMax=1.4}}
 {id=Modulator_11, children [Random_9], properties {pitchMax=1.1, pitchMin=0.9}}
 {id=Modulator_2, children [Random_2], properties {pitchMax=1.1, pitchMin=0.9}}
 {id=Modulator_4, children [Random_3], properties {pitchMax=1.4}}
 {id=Modulator_5, children [Random_0]}
 {id=Modulator_7, children [Random_6], properties {pitchMax=1.4}}
 {id=Modulator_8, children [Random_1]}
 {id=QualityLevel_0, children [Mixer_3,Mixer_5,Mixer_1,Mixer_7]}
 {
  id=Random_0
  children [
   WavePlayer_0
   WavePlayer_2
   WavePlayer_3
   WavePlayer_5
   WavePlayer_6
   WavePlayer_7
   WavePlayer_9
   WavePlayer_10
   WavePlayer_11
   WavePlayer_12
  ]
  properties {weights [1,1,1,1,1,1,1,1,1,1]}
 }
 {
  id=Random_1
  children [WavePlayer_13, WavePlayer_14, WavePlayer_47, WavePlayer_51, WavePlayer_52, WavePlayer_54, WavePlayer_55]
  properties {weights [1,1,1,1,1,1,1]}
 }
 {id=Random_2, children [WavePlayer_56, WavePlayer_57, WavePlayer_58], properties {weights [1,1,1]}}
 {id=Random_3, children [WavePlayer_19, WavePlayer_18, WavePlayer_20, WavePlayer_21], properties {weights [1,1,1,1]}}
 {id=Random_6, children [WavePlayer_37, WavePlayer_36, WavePlayer_38, WavePlayer_39], properties {weights [1,1,1,1]}}
 {id=Random_9, children [WavePlayer_59, WavePlayer_60, WavePlayer_61], properties {weights [1,1,1]}}
 {id=WavePlayer_0, properties {soundWaveAssetPtr=&/Demo_Grecian_Urn_Impact_01}}
 {id=WavePlayer_1, properties {soundWaveAssetPtr=/Game/Sounds/Fort_Impact_Sounds/Demolition/Defaults/Rock_Distant_Impact_02}}
 {id=WavePlayer_10, properties {soundWaveAssetPtr=&/Demo_Grecian_Urn_Impact_08}}
 {id=WavePlayer_11, properties {soundWaveAssetPtr=&/Demo_Grecian_Urn_Impact_09}}
 {id=WavePlayer_12, properties {soundWaveAssetPtr=&/Demo_Grecian_Urn_Impact_10}}
 {id=WavePlayer_13, properties {soundWaveAssetPtr=&Demo_Grecian_Urn_Impact_01}}
 {id=WavePlayer_14, properties {soundWaveAssetPtr=&/Demo_Grecian_Urn_Impact_02}}
 {id=WavePlayer_15, properties {soundWaveAssetPtr=&Rock_Distant_Impact_02}}
 {id=WavePlayer_18, properties {soundWaveAssetPtr=/Game/Sounds/Fort_Impact_Sounds/Demolition/Defaults/Rock_Distant_Impact_10}}
 {id=WavePlayer_19, properties {soundWaveAssetPtr=&Rock_Distant_Impact_02}}
 {id=WavePlayer_2, properties {soundWaveAssetPtr=&Demo_Grecian_Urn_Impact_02}}
 {id=WavePlayer_20, properties {soundWaveAssetPtr=/Game/Sounds/Fort_Impact_Sounds/Demolition/Defaults/Rock_Distant_Impact_03}}
 {id=WavePlayer_21, properties {soundWaveAssetPtr=/Game/Sounds/Fort_Impact_Sounds/Demolition/Defaults/Rock_Distant_Impact_07}}
 {id=WavePlayer_3, properties {soundWaveAssetPtr=&/Demo_Grecian_Urn_Impact_03}}
 {id=WavePlayer_36, properties {soundWaveAssetPtr=&Rock_Distant_Impact_10}}
 {id=WavePlayer_37, properties {soundWaveAssetPtr=&Rock_Distant_Impact_02}}
 {id=WavePlayer_38, properties {soundWaveAssetPtr=&Rock_Distant_Impact_03}}
 {id=WavePlayer_39, properties {soundWaveAssetPtr=&Rock_Distant_Impact_07}}
 {id=WavePlayer_47, properties {soundWaveAssetPtr=&Demo_Grecian_Urn_Impact_03}}
 {id=WavePlayer_5, properties {soundWaveAssetPtr=&/Demo_Grecian_Urn_Impact_04}}
 {id=WavePlayer_51, properties {soundWaveAssetPtr=&Demo_Grecian_Urn_Impact_04}}
 {id=WavePlayer_52, properties {soundWaveAssetPtr=&/Demo_Grecian_Urn_Impact_05}}
 {id=WavePlayer_54, properties {soundWaveAssetPtr=&/Demo_Grecian_Urn_Impact_06}}
 {id=WavePlayer_55, properties {soundWaveAssetPtr=&/Demo_Grecian_Urn_Impact_07}}
 {id=WavePlayer_56, properties {soundWaveAssetPtr=&Demo_Grecian_Urn_Impact_03}}
 {id=WavePlayer_57, properties {soundWaveAssetPtr=&Demo_Grecian_Urn_Impact_02}}
 {id=WavePlayer_58, properties {soundWaveAssetPtr=&Demo_Grecian_Urn_Impact_01}}
 {id=WavePlayer_59, properties {soundWaveAssetPtr=&Demo_Grecian_Urn_Impact_03}}
 {id=WavePlayer_6, properties {soundWaveAssetPtr=&Demo_Grecian_Urn_Impact_05}}
 {id=WavePlayer_60, properties {soundWaveAssetPtr=&Demo_Grecian_Urn_Impact_02}}
 {id=WavePlayer_61, properties {soundWaveAssetPtr=&Demo_Grecian_Urn_Impact_01}}
 {id=WavePlayer_7, properties {soundWaveAssetPtr=&Demo_Grecian_Urn_Impact_06}}
 {id=WavePlayer_9, properties {soundWaveAssetPtr=&Demo_Grecian_Urn_Impact_07}}
]
properties {
 maxDistance=8000
 soundClassObject=&Fort_Impacts
 volumeMultiplier=0.796
 concurrencyOverrides {maxCount=4}
 concurrencySet [/Game/Sounds/Concurrency_Rules/Demo_FourMaximum_StopFurthestOldest_VolumeScale]
}
